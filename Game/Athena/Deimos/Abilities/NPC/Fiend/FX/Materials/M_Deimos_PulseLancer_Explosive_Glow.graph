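# Translucent unlit glow material for Niagara mesh particles, named for the Deimos Fiend PulseLancer explosive.
# A squared radial UV falloff, depth-faded above low quality, drives both colour and opacity.
# Colour pulses on a time sine into shared emissive adjustments, while opacity scales by time of day with optional extra depth fade.
# World position offset aligns the mesh to camera, blended by a distance fade, plus a push along the actor-to-camera axis.
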Material /Game/Athena/Deimos/Abilities/NPC/Fiend/FX/Materials/M_Deimos_PulseLancer_Explosive_Glow
graph {
 properties {
  bUsedWithNiagaraMeshParticles=true
  blendMode=BLEND_Translucent
  pixelDepthOffsetMode=PDOM_Legacy
  shadingModel=MSM_Unlit
  usedShadingModels=Unlit
 }
 nodes [
  n1 = VectorParameter() {parameterName=RotationA, defaultValue=<rgba 0 1 0 1>}
  n2 = ScalarParameter() {defaultValue=3400, parameterName=End Fade}
  n3 = ScalarParameter() {defaultValue=3000, parameterName=Start Fade}
  n4 = Subtract(A=CameraPositionWS(), B=ActorPositionWS())
  Add(
   A=LinearInterpolate(
    A=MaterialFunctionCall!(WorldPosition, PivotLocation, WS Normals, Custom Object Basis 1, Custom Object Basis 2=@n1.RGB, Custom Object Basis 3) {materialFunction=/Engine/Functions/Engine_MaterialFunctions02/WorldPositionOffset/AlignMeshToTheCamera}.World Position Offset
    Alpha=MaterialFunctionCall(Fade_Start_Distance=@n2, Fade_End_Distance=@n3) {materialFunction=&MF_Distance_SyncedFade_FX}.Result
   ) {constB=0}
   B=Multiply(A=Normalize(VectorInput=@n4), B=ScalarParameter() {defaultValue=-15, parameterName=CameraOffsetDark})
  )
  n5 = Multiply(
   A=MaterialFunctionCall(
    Value_Morning=ScalarParameter() {defaultValue=0.8, parameterName="Value_Morning (S)"}
    Value_Day=ScalarParameter() {defaultValue=1, parameterName="Value_Day (S)"}
    Value_Evening=ScalarParameter() {defaultValue=0.8, parameterName="Value_Evening (S)"}
    Value_Night=ScalarParameter() {defaultValue=0.6, parameterName="Value_Night (S)"}
   ) {materialFunction=&Scalar_Time_of_Day}.Result
   B=Multiply(A=NamedRerouteUsage(), B=ScalarParameter() {defaultValue=1, parameterName=OpacityDarkBG})
  )
  LinearInterpolate(
   A=@n5
   B=DepthFade(Opacity=@n5, FadeDistance=ScalarParameter() {defaultValue=15, group=DepthFade, parameterName=FadeDistance})
   Alpha=ScalarParameter() {group=DepthFade, parameterName=UseDF}
  )
  n6 = Multiply(
   A=VectorParameter() {parameterName=ColorDarkGlow, defaultValue=<rgba 2 0.1053 0.04 1>}.RGB
   B=Multiply(A=NamedRerouteUsage(), B=ScalarParameter() {defaultValue=1, parameterName=HoloIntensity})
  )
  n7 = Time()
  n8 = ScalarParameter() {defaultValue=3.6, group=Deimos_Emissive, parameterName=Emissive_Pulsing_PulseSpeed}
  n9 = Multiply(
   A=Sine(None=Divide(A=@n7, B=@n8))
   B=ScalarParameter() {defaultValue=0.45, group=Deimos_Emissive, parameterName=Emissive_Pulsing_Pulse Intensity}
  ) {constB=0.1}
  MaterialFunctionCall(Emissive=Multiply(A=ShadingPathSwitch(Default=@n6, Mobile=@n6) {desc=Migrated from FeatureLevelSwitch}, B=Add(A=@n9))) {materialFunction=&MF_FX_EmissiveAdjustments}
  n10 = ScalarParameter() {defaultValue=1, parameterName=BG Boost}
  n11 = Distance(A=TextureCoordinate(), B=Constant() {r=0.5})
  n12 = ScalarParameter() {defaultValue=2, parameterName=GlowSize}
  n13 = Saturate(None=OneMinus(None=Multiply(A=@n11, B=@n12) {constB=2}))
  n14 = Multiply(A=@n13, B=@n13)
  n15 = Constant() {r=1}
  n16 = DepthFade(FadeDistance=ScalarParameter() {defaultValue=25, parameterName=Depth Fade Distance})
  NamedRerouteDeclaration(None=Multiply(A=Multiply(A=@n10, B=@n14), B=QualitySwitch(Default=@n15, Low=@n15, High=@n16, Medium=@n16))) {name=HolographicMesh, nodeColor=<rgba 0.6172 0 1 1>}
 ]
}
