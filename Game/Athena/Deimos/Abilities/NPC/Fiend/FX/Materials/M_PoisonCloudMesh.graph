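# Unlit translucent material for Niagara mesh particles, drawing a poison cloud for the Fiend NPC ability in Deimos.
# A tiled infected-pattern texture pans over time with normal-map UV distortion, and the normal dotted with a direction blends two emissive colors.
# Opacity combines texture, depth fade (skipped at Low quality), near-camera fade, vertical edge fade, angle/Fresnel fade and particle alpha.
# A separately panned normal map and gradient build a vertex-normal displacement for WPO.
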Material /Game/Athena/Deimos/Abilities/NPC/Fiend/FX/Materials/M_PoisonCloudMesh
graph {
 properties {
  bUsedWithNiagaraMeshParticles=true
  blendMode=BLEND_Translucent
  pixelDepthOffsetMode=PDOM_Legacy
  shadingModel=MSM_Unlit
  usedShadingModels=Unlit
 }
 nodes [
  n1 = Subtract(A=SceneDepth().R, B=PixelDepth().R)
  n2 = ScalarParameter() {defaultValue=1024, parameterName=Fog Distance}
  n3 = Multiply(
   A=TextureCoordinate()
   B=AppendVector(A=ScalarParameter() {defaultValue=1.5, parameterName=UV Tile X}, B=ScalarParameter() {defaultValue=0.1875, parameterName=UV Tile Y})
  )
  n4 = TextureSample(UVs=@n3) {samplerType=SAMPLERTYPE_Normal, texture=/Game/Global/Textures/Default/Tileable_Patterns/2ndInfectedMapNormal}
  n5 = Time()
  n6 = AppendVector(A=ScalarParameter() {parameterName=UV Pan X}, B=ScalarParameter() {defaultValue=0.05, parameterName=UV Pan Y})
  n7 = Add(A=Multiply(A=ComponentMask(None=@n4.RGB) {g=true, r=true}) {constB=0.02}, B=Add(A=@n3, B=Multiply(A=@n5, B=@n6)))
  n8 = TextureSample(UVs=@n7) {texture=&2ndInfectedMapLinear}
  n9 = ScalarParameter() {defaultValue=100, parameterName=Depth Fade 2}
  n10 = ScalarParameter() {defaultValue=50, parameterName=Depth Fade}
  n11 = TextureSample(UVs=@n7) {texture=&2ndInfectedMapLinear}
  n12 = Multiply(A=@n8.G, B=DepthFade(FadeDistance=LinearInterpolate(A=@n9, B=@n10, Alpha=@n11.G)))
  Add(A=Clamp(None=Divide(A=@n1, B=@n2) {constB=1024}), B=@n12)
  n13 = TextureSample(UVs=@n7) {samplerType=SAMPLERTYPE_Normal, texture=&2ndInfectedMapNormal}
  n14 = Normalize(VectorInput=VectorParameter() {parameterName=Color Lerp Direction, defaultValue=<rgba 0 0 1 1>}.RGB)
  MaterialFunctionCall(Emissive=LinearInterpolate(
    A=VectorParameter() {parameterName=Color_01, defaultValue=<rgba 0.9 0.5694 0.112 1>}.RGB
    B=VectorParameter() {parameterName=Color_02, defaultValue=<rgba 0.2656 0.255 0.1073 1>}.RGB
    Alpha=DotProduct(A=@n13.RGB, B=@n14)
   )) {materialFunction=&MF_FX_EmissiveAdjustments}
  n15 = MaterialFunctionCall() {materialFunction=&LinearGradient}
  n16 = OneMinus(None=@n15.VGradient)
  n17 = ScalarParameter() {defaultValue=0.5, parameterName=WPO UV Tile X}
  n18 = ScalarParameter() {defaultValue=0.0625, parameterName=WPO UV Tile Y}
  n19 = ScalarParameter() {parameterName=WPO UV Pan X}
  n20 = ScalarParameter() {defaultValue=0.05, parameterName=WPO UV Pan Y}
  n21 = Add(A=Multiply(A=TextureCoordinate(), B=AppendVector(A=@n17, B=@n18)), B=Multiply(A=Time(), B=AppendVector(A=@n19, B=@n20)))
  Multiply(
   A=Multiply(A=Power(Base=@n16) {constExponent=1.4}, B=ScalarParameter() {defaultValue=24, parameterName=WPO Distortion Multiplier}) {constB=35}
   B=Multiply(
    A=VertexNormalWS()
    B=TextureSample(UVs=@n21) {samplerType=SAMPLERTYPE_Normal, texture=&2ndInfectedMapNormal}.RGB
   )
  ) {constB=5}
  n22 = Multiply(
   A=QualitySwitch(Default=@n12, Low=@n8.G, High=@n12, Medium=@n12, Epic=@n12)
   B=MaterialFunctionCall(Distance=ScalarParameter() {defaultValue=250, parameterName=Near Camera Distance}) {materialFunction=&Near_Camera_Fade}.Result
  )
  n23 = Multiply(A=@n15.VGradient, B=@n16)
  n24 = ScalarParameter() {defaultValue=2, parameterName=s}
  n25 = Clamp(None=Multiply(A=Power(Base=@n23, Exp=@n24), B=ScalarParameter() {defaultValue=10, parameterName=sssf}))
  n26 = ScalarParameter() {defaultValue=1.4, parameterName=Angle Fade Bias}
  n27 = Power(Base=DotProduct(A=VertexNormalWS(), B=CameraVectorWS()), Exp=@n26)
  n28 = MaterialFunctionCall(Axis=Constant3Vector() {constant=<rgba 0 0 1 0>}, Exponent=@n26) {materialFunction=&AxisAlignedFresnel}
  n29 = ScalarParameter() {defaultValue=1, parameterName=UseAxisFresnel?}
  Multiply(A=ParticleColor().A, B=Multiply(A=Multiply(A=@n22, B=@n25), B=LinearInterpolate(A=@n27, B=@n28.Fresnel, Alpha=@n29)))
 ]
}
