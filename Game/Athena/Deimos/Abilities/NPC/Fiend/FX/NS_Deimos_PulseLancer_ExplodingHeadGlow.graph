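# Niagara system for the Deimos Fiend NPC that, by its name, provides the glow on a Pulse Lancer's exploding head.
# Its one emitter, LightAndGlow001, burst-spawns scaled, oriented mesh particles and has only a mesh renderer, despite the "Light" in its name.
# Every update re-samples a skeletal mesh location with a small offset, so the mesh follows the character.
# Particles are not killed when their lifetime elapses.
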NiagaraSystem /Game/Athena/Deimos/Abilities/NPC/Fiend/FX/NS_Deimos_PulseLancer_ExplodingHeadGlow
dependencies {
 usedDataInterfaces [&NiagaraDataInterfaceSkeletalMesh]
 usedDynamicInputs []
 usedModules [
  &self:SystemSpawnScript.SystemScriptSource.SystemScriptGraph.NiagaraNodeAssignment_0.SetVariables_#1
  &SystemState
 ]
}
emitters {LightAndGlow001 {
  emitterUpdateScript [
   &EmitterState(
    Inactive Response=::NewEnumerator1
    Life Cycle Mode=::NewEnumerator1
    Loop Behavior=::NewEnumerator1
    Loop Duration Mode=::NewEnumerator1
    UseLoopDelay=false
   )
   &SpawnBurst_Instantaneous(Use Spawn Probability=false, UseLoopCountLimit=false)
  ]
  particleSpawnScript [
   &InitializeParticle(Lifetime=5, Mesh Scale=<0.7 0.85 0.7>, Mesh Scale Mode=::NewEnumerator4, UsePositionOffset=false, Write Material Random=true)
   &InitialMeshOrientation(Mesh Orientation Mode=::NewEnumerator7, Use Rotation Vector=true)
  ]
  particleUpdateScript [
   &ParticleState(Kill Particles When Lifetime Has Elapsed=false, Let Infinitely Lived Particles Die When Emitter Deactivates=true)
   &SkeletalMeshLocation(
    MaskSpawnGroup=false
    Mesh Sampling Type=::NewEnumerator5
    Orientation Sampling=::NewEnumerator1
    Sampled Position Offset=<10 0 3>
    Skeletal Mesh=$System.Skeletal Mesh
    Socket Sampling Mode=::NewEnumerator5
   )
  ]
  rendererClasses [&NiagaraMeshRendererProperties]
  renderers [{rendererClass=&NiagaraMeshRendererProperties, rendererIndex=0}]
 }}
summary {
 systemName=NS_Deimos_PulseLancer_ExplodingHeadGlow
 emitters [{emitterName=LightAndGlow001, rendererClasses [&NiagaraMeshRendererProperties]}]
 userVariables []
}
